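# M_GrassBit is a masked, default-lit material for Niagara sprite grass bits in the main player's Traction emote effects.
# It multiplies the grass stalk diffuse texture by particle color, letting the particle system tint each sprite.
# A separate branch adds a small constant to the grass stalk normal map sample and saturates the result.
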Material /Game/Animation/Game/MainPlayer/Emotes/Traction/FX/M_GrassBit
graph {
 properties {bUsedWithNiagaraSprites=true, blendMode=BLEND_Masked, pixelDepthOffsetMode=PDOM_Legacy, usedShadingModels=Default Lit}
 nodes [
  Multiply(A=ParticleColor().RGB, B=TextureSample() {texture=&/T_GrassStalk_Bit_D}.RGB)
  Saturate(None=Add(A=TextureSample() {samplerType=SAMPLERTYPE_Normal, texture=&/T_GrassStalk_Bit_N}.RGB) {constB=0.2})
 ]
}
