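# Niagara exhaust system for the Traction emote, throwing grass debris and smoke wisps from a skeletal mesh.
# Three emitters spawn at sampled mesh locations: mesh grass, GPU sprite grass, and SubUV-animated smoke sprites, each launched with cone velocity and pulled down by gravity.
# Spawn-per-unit rates follow movement, scale by engine quality level and drop to zero in water.
# Front-end preview swaps these for constant spawn rates.
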NiagaraSystem /Game/Animation/Game/MainPlayer/Emotes/Traction/FX/NS_Traction_Exhaust
dependencies {
 usedDataInterfaces [
  &NiagaraDataInterfaceArrayFloat
  &NiagaraDataInterfaceColorCurve
  &NiagaraDataInterfaceCurve
  &NiagaraDataInterfaceVectorCurve
 ]
 usedDynamicInputs [&RandomRangeLinearColor]
 usedModules [
  &self:SystemSpawnScript.SystemScriptSource.SystemScriptGraph.NiagaraNodeAssignment_0.SetVariables_#1
  &SystemState
  &Timeline
 ]
}
emitters {
 Grass {
  emitterUpdateScript [
   &EmitterState()
   &SpawnPerUnit(
    Module.Use Max Movement Threshold=true
    Module.Use Movement Tolerance=true
    Movement Tolerance=1
    Spawn Per Unit:NiagaraFloat=&MakeCustomFloatFromBool(
     Bool=$User.bIsFrontEndPreview
     True Float=0
     False Float=&Multiply_Float(
      A=&MakeCustomFloatFromBool(Bool=$User.bIsInWater, True Float=0, False Float=1)
      B=&SelectFloatFromArray(Direct Array Index=$Engine.QualityLevel, Float Selection Array=DataInterface() {propertyValues {FloatData [5,4,2,1,1]}})
     )
    )
    Use Spawn Probability=false
   )
   &SpawnRate(
    SpawnRate=&MakeCustomFloatFromBool(Bool=$User.bIsFrontEndPreview, True Float=100, False Float=0)
    Use Spawn Probability=false
   )
  ]
  particleSpawnScript [
   /SetVariables(Particles.MeshIndex=&RandomRangeInt(Minimum=0, Maximum=1)) {isSetParameters=true}
   &SkeletalMeshLocation(MaskSpawnGroup=false, Mesh Sampling Type=::NewEnumerator5, Skeletal Mesh=$User.SkeletalMesh, Socket Sampling Mode=::NewEnumerator5)
   /Niagara/Modules/Spawn/Initialization/InitializeParticle(
    Color=$System.GrassBaseColor
    Lifetime=&UniformRangedFloat(Minimum=0.25, Maximum=0.5)
    Mass=10
    Mesh Scale=&RandomRangeVector(Minimum=<0.1 0.2 0.2>, Maximum=<0.25 0.4 0.4>)
    Module.Write Color=true
    Module.Write Lifetime=true
    Module.Write Mass=true
    Module.Write Material Random=true
    Module.Write Position=false
    Module.Write Scale=true
    Module.Write SpriteRotation=false
    Module.Write SpriteSize=false
   )
   &AddVelocity(Velocity=&RandomRangeVector(Minimum=<0 -100 0>, Maximum=<200 -400 50>))
   &InheritVelocity(
    Module.Limit Speed=true
    Velocity Limit=50
    Velocity Scale=&RandomRangeVector(Minimum=<0.5 0.5 0.5>, Maximum=<1 1 1>)
   )
   &AddVelocityInCone(
    Cone Angle=50
    Cone Axis=<0 -1 1>
    Module.Use Velocity Falloff On Cone Axis=false
    Velocity Strength=&UniformRangedFloat(Minimum=0, Maximum=50)
   )
   &InitialMeshOrientation(Use Rotation Vector=false)
   /Niagara/Modules/Spawn/Location/SphereLocation(
    MaskSpawnGroup=false
    Module.Override Randomness=false
    Sphere Origin:NiagaraPosition=&ConvertVectorToPosition(
     Input Position=Linked() {linkedVariable {name=Particles.SkeletalMeshLocation.SampledPosition, type=NiagaraPosition}}
     Vector Is In Absolute Large World Space:NiagaraBool=false
    )
    Sphere Radius=5
    Surface Only=false
   ) {module=SphereLocation001}
  ]
  particleUpdateScript [
   &ParticleState()
   &Drag(Use Linear Drag=true, Use Rotational Drag=false)
   &ScaleVelocity(
    Coordinate Space:ENiagaraCoordinateSpace=::Simulation
    Velocity Scale=&VectorFromCurve(
     VectorCurve=DataInterface() {propertyValues {
       XCurve {keys [
         {arriveTangent=0.6954, interpMode=RCIM_Cubic, leaveTangent=0.6954, tangentMode=RCTM_User, time=0, value=0.5}
         {arriveTangent=-0.5, interpMode=RCIM_Cubic, leaveTangent=-0.5, time=0.3575, value=1.0797}
         {arriveTangent=-4.0987, interpMode=RCIM_Cubic, leaveTangent=-4.0987, tangentMode=RCTM_User, time=1, value=0}
        ]}
       YCurve {keys [
         {arriveTangent=0.6954, interpMode=RCIM_Cubic, leaveTangent=0.6954, tangentMode=RCTM_User, time=0, value=0.5}
         {arriveTangent=-0.5, interpMode=RCIM_Cubic, leaveTangent=-0.5, time=0.3575, value=1.0797}
         {arriveTangent=-4.0987, interpMode=RCIM_Cubic, leaveTangent=-4.0987, tangentMode=RCTM_User, time=1, value=0}
        ]}
       ZCurve {keys [
         {tangentMode=RCTM_User, time=0, value=1}
         {arriveTangent=0.9778, interpMode=RCIM_Cubic, leaveTangent=0.9778, tangentMode=RCTM_User, time=0.1, value=1}
         {arriveTangent=-22.649, interpMode=RCIM_Cubic, leaveTangent=-22.649, tangentMode=RCTM_User, time=1, value=-2}
        ]}
      }}
     Scale Curve=<1 1 0.5>
    )
   ) {enabled=false}
   &GravityForce(Gravity=<0 0 -720>)
   &CurlNoiseForce(Noise Frequency=100, Noise Strength=50, Pan Noise=false)
   &SolveForcesAndVelocity(Clamp Velocity=false, Limit Acceleration=false)
   &ScaleColor(
    Scale Alpha=1.2
    Scale RGB=&MakeVectorFromLinearColorRGB(Color=&ColorFromCurve(ColorCurve=DataInterface() {propertyValues {
        AlphaCurve {keys []}
        BlueCurve {keys [<at 0.0081 0.6244>,<at 0.1423 0.6305>,<at 0.9559 1>]}
        GreenCurve {keys [<at 0.0081 0.8353>,<at 0.1423 0.7629>,<at 0.9559 0.8821>]}
        RedCurve {keys [<at 0.0081 1>,<at 0.1423 1>,<at 0.9559 0.8145>]}
       }}))
    ScaleA=true
    ScaleRGB=true
   )
   &UpdateMeshOrientation(Rotation Rate=&RandomRangeFloat(Minimum=-0.4, Maximum=0.4), Rotation Vector=<1 1 1>)
   &ScaleMeshSize(Scale Factor=&VectorFromFloat(Value=&FloatFromCurve(FloatCurve=DataInterface() {propertyValues {Curve {keys [
          {arriveTangent=8.5765, interpMode=RCIM_Cubic, leaveTangent=8.5765, tangentMode=RCTM_User, time=0.1, value=0.7}
          {arriveTangent=0.65, interpMode=RCIM_Cubic, leaveTangent=0.65, time=0.2, value=1}
          {arriveTangent=-1.25, interpMode=RCIM_Cubic, leaveTangent=-1.25, time=0.6, value=0.625}
          {arriveTangent=0.095, interpMode=RCIM_Cubic, leaveTangent=0.095, tangentMode=RCTM_User, time=1, value=0}
         ]}}})))
   &ScaleSpriteSizeBySpeed(
    Max Scale Factor=&Vector2DFromFloat(Value=1)
    Min Scale Factor=&Vector2DFromFloat(Value=0.25)
    Velocity Threshold=200
   )
  ]
  rendererClasses [&NiagaraMeshRendererProperties]
  renderers [{rendererClass=&NiagaraMeshRendererProperties, rendererIndex=0}]
 }
 Grass001 {
  simTarget=GPUComputeSim
  emitterUpdateScript [
   &EmitterState()
   &SpawnPerUnit(
    Module.Use Max Movement Threshold=true
    Module.Use Movement Tolerance=true
    Movement Tolerance=1
    Spawn Per Unit:NiagaraFloat=&MakeCustomFloatFromBool(
     Bool=$User.bIsFrontEndPreview
     True Float=0
     False Float=&Multiply_Float(
      A=&MakeCustomFloatFromBool(Bool=$User.bIsInWater, True Float=0, False Float=1)
      B=&SelectFloatFromArray(Direct Array Index=$Engine.QualityLevel, Float Selection Array=DataInterface() {propertyValues {FloatData [2,1.5,1,0.5,0.5]}})
     )
    )
    Use Spawn Probability=false
   )
   &SpawnRate(
    SpawnRate=&MakeCustomFloatFromBool(Bool=$User.bIsFrontEndPreview, True Float=500, False Float=0)
    Use Spawn Probability=false
   )
  ]
  particleSpawnScript [
   /SetVariables(Particles.SubImageIndex=&RandomRangeFloat(Minimum=0, Maximum=6)) {isSetParameters=true}
   &SkeletalMeshLocation(MaskSpawnGroup=false, Mesh Sampling Type=::NewEnumerator5, Skeletal Mesh=$User.SkeletalMesh, Socket Sampling Mode=::NewEnumerator5)
   &InitializeParticle(
    Color=<rgba 0.5 0.5 0.5 1>
    Lifetime=&UniformRangedFloat(Minimum=0.4, Maximum=0.6)
    Mass=10
    Module.Write Color=true
    Module.Write Lifetime=true
    Module.Write Mass=true
    Module.Write Material Random=true
    Module.Write Position=false
    Module.Write Scale=false
    Module.Write SpriteRotation=true
    Module.Write SpriteSize=true
    Sprite Rotation=&RandomRangeFloat(Minimum=-360, Maximum=360)
    Sprite Size=&RandomRangeVector2D(Minimum=<10 10>, Maximum=<20 20>)
   )
   &AddVelocity(Velocity=&RandomRangeVector(Minimum=<0 -100 0>, Maximum=<250 -400 50>))
   &InheritVelocity(Module.Limit Speed=true, Velocity Scale=&RandomRangeVector(Minimum=<0.5 0.5 0.5>, Maximum=<1 1 1>))
   &AddVelocityInCone(
    Cone Angle=50
    Cone Axis=<0 -1 1>
    Module.Use Velocity Falloff On Cone Axis=false
    Velocity Strength=&UniformRangedFloat(Minimum=0, Maximum=100)
   )
   &SphereLocation(
    MaskSpawnGroup=false
    Module.Override Randomness=false
    Sphere Origin:NiagaraPosition=&ConvertVectorToPosition(
     Input Position=Linked() {linkedVariable {name=Particles.SkeletalMeshLocation.SampledPosition, type=NiagaraPosition}}
     Vector Is In Absolute Large World Space:NiagaraBool=false
    )
    Sphere Radius=5
    Surface Only=false
   ) {module=SphereLocation001}
  ]
  particleUpdateScript [
   &ParticleState()
   &Drag(Use Linear Drag=true, Use Rotational Drag=false)
   &ScaleVelocity(
    Coordinate Space:ENiagaraCoordinateSpace=::Simulation
    Velocity Scale=&VectorFromCurve(
     VectorCurve=DataInterface() {propertyValues {
       XCurve {keys [
         {arriveTangent=0.6954, interpMode=RCIM_Cubic, leaveTangent=0.6954, tangentMode=RCTM_User, time=0, value=0.5}
         {arriveTangent=-0.5, interpMode=RCIM_Cubic, leaveTangent=-0.5, time=0.3575, value=1.0797}
         {arriveTangent=-4.0987, interpMode=RCIM_Cubic, leaveTangent=-4.0987, tangentMode=RCTM_User, time=1, value=0}
        ]}
       YCurve {keys [
         {arriveTangent=0.6954, interpMode=RCIM_Cubic, leaveTangent=0.6954, tangentMode=RCTM_User, time=0, value=0.5}
         {arriveTangent=-0.5, interpMode=RCIM_Cubic, leaveTangent=-0.5, time=0.3575, value=1.0797}
         {arriveTangent=-4.0987, interpMode=RCIM_Cubic, leaveTangent=-4.0987, tangentMode=RCTM_User, time=1, value=0}
        ]}
       ZCurve {keys [
         {tangentMode=RCTM_User, time=0, value=1}
         {arriveTangent=0.9778, interpMode=RCIM_Cubic, leaveTangent=0.9778, tangentMode=RCTM_User, time=0.1, value=1}
         {arriveTangent=1.6152, interpMode=RCIM_Cubic, leaveTangent=1.6152, tangentMode=RCTM_User, time=1, value=-2}
        ]}
      }}
     Scale Curve=<1 1 0.75>
    )
   ) {enabled=false}
   &GravityForce(Gravity=<0 0 -720>)
   &CurlNoiseForce(Noise Frequency=100, Noise Strength=100, Pan Noise=false)
   &SolveForcesAndVelocity(Clamp Velocity=false, Limit Acceleration=false)
   &ScaleColor(
    Scale Alpha=1.2
    Scale RGB=&MakeVectorFromLinearColorRGB(Color=&ColorFromCurve(ColorCurve=DataInterface() {propertyValues {
        AlphaCurve {keys []}
        BlueCurve {keys [<at 0.0081 0.6244>,<at 0.1423 0.6305>,<at 0.9559 1>]}
        GreenCurve {keys [<at 0.0081 0.8353>,<at 0.1423 0.7629>,<at 0.9559 0.8821>]}
        RedCurve {keys [<at 0.0081 1>,<at 0.1423 1>,<at 0.9559 0.8145>]}
       }}))
    ScaleA=true
    ScaleRGB=true
   )
   &ScaleSpriteSize(Uniform Curve Sprite Scale=DataInterface() {propertyValues {Curve {keys [
        {arriveTangent=8.5765, interpMode=RCIM_Cubic, leaveTangent=8.5765, tangentMode=RCTM_User, time=0.1, value=0.5}
        {arriveTangent=-0.5096, interpMode=RCIM_Cubic, leaveTangent=-0.5096, time=0.2, value=1}
        {arriveTangent=-5.3839, interpMode=RCIM_Cubic, leaveTangent=-5.3839, tangentMode=RCTM_User, time=0.7622, value=0.1625}
        {arriveTangent=0.095, interpMode=RCIM_Cubic, leaveTangent=0.095, tangentMode=RCTM_User, time=0.8, value=-0.05}
       ]}}})
   &SpriteRotationRate(Rotation Rate=&RandomRangeFloat(Minimum=-360, Maximum=360))
  ]
  rendererClasses [&NiagaraSpriteRendererProperties]
  renderers [{rendererClass=&NiagaraSpriteRendererProperties}]
 }
 Smoke_Wisp {
  emitterUpdateScript [
   &EmitterState()
   &SpawnPerUnit(
    Max Movement Threshold=1000
    Module.Use Max Movement Threshold=true
    Module.Use Movement Tolerance=true
    Spawn Per Unit:NiagaraFloat=&MakeCustomFloatFromBool(
     Bool=$User.bIsFrontEndPreview
     True Float=0
     False Float=&Multiply_Float(
      A=&MakeCustomFloatFromBool(Bool=$User.bIsInWater, True Float=0, False Float=1)
      B=&SelectFloatFromArray(Direct Array Index=$Engine.QualityLevel, Float Selection Array=DataInterface() {propertyValues {FloatData [15,10,7.5,7.5,7.5]}})
     )
    )
    Use Spawn Probability=false
   )
   &SpawnRate(
    SpawnRate=&MakeCustomFloatFromBool(Bool=$User.bIsFrontEndPreview, True Float=25, False Float=0)
    Use Spawn Probability=false
   )
  ]
  particleSpawnScript [
   /Niagara/Modules/Spawn/Initialization/V2/InitializeParticle(
    Color=<rgba 0.75 0.75 1 0.25>
    Lifetime=&RandomRangeFloat(Minimum=0.4, Maximum=0.6)
    Position Mode=::NewEnumerator0
    Sprite Rotation Mode=::NewEnumerator2
    Sprite Size Mode=::NewEnumerator3
    Uniform Sprite Size=&MaskFloatBySpawnGroup(
     Masked Float=&RandomRangeFloat(Minimum=100, Maximum=150)
     Passthrough Float=&RandomRangeFloat(Minimum=100, Maximum=200)
    )
    Write Material Random=true
   )
   &SkeletalMeshLocation(MaskSpawnGroup=false, Mesh Sampling Type=::NewEnumerator5, Skeletal Mesh=$User.SkeletalMesh, Socket Sampling Mode=::NewEnumerator5)
   &TorusLocation(
    Align Torus With Axis=false
    Large Radius=20
    MaskSpawnGroup=false
    Torus Distribution Mode:ENiagaraTorusDistributionMode=::NewEnumerator1
    Torus Mode=::NewEnumerator2
    U Position:NiagaraFloat=&ReturnNormalizedExecIndex(Normalized Index Includes One=false)
   ) {enabled=false}
   &SphereLocation(MaskSpawnGroup=false, Module.Override Randomness=false, Sphere Radius=1, Surface Only=false) {module=SphereLocation001}
   &AddVelocityInCone(
    Cone Angle=50
    Cone Axis=<0 -1 1>
    Module.Use Velocity Falloff On Cone Axis=false
    Velocity Strength=&UniformRangedFloat(Minimum=0, Maximum=500)
   )
   &AddVelocity(Velocity=&MaskVectorBySpawnGroup(
     Masked Vector=&RandomRangeVector(Minimum=<25 -150 0>, Maximum=<50 -300 0>)
     Passthrough Vector=&RandomRangeVector(Minimum=<25 -150 0>, Maximum=<50 -300 0>)
    )) {module=AddVelocity002}
   /SetVariables(Particles.InitSpriteRotRate=&UniformRangedFloat(Minimum=-3, Maximum=3)) {isSetParameters=true}
   /SetVariables(Particles.RandomFloat=&RandomRangeFloat(Minimum=0, Maximum=1)) {isSetParameters=true}
   /SetVariables(Particles.SubUVRandomStart:NiagaraInt32=&RandomRangeInt(Minimum=0, Maximum=44)) {isSetParameters=true}
  ]
  particleUpdateScript [
   &ParticleState()
   &ScaleColor(
    Scale Alpha=&FloatFromCurve(FloatCurve=DataInterface() {propertyValues {Curve {keys [{arriveTangent=-3.3421, interpMode=RCIM_Cubic, leaveTangent=-3.3421, tangentMode=RCTM_User, time=0.1219, value=1}, <at 1 0 RCIM_Cubic>]}}})
    ScaleA=true
    ScaleRGB=true
   )
   &SpriteRotationRate(Rotation Rate=&Multiply_Float(
     A=$Particles.InitSpriteRotRate
     B=&FloatFromCurve(
      FloatCurve=DataInterface() {propertyValues {Curve {keys [
          {arriveTangent=-4.0107, interpMode=RCIM_Cubic, leaveTangent=-4.0107, tangentMode=RCTM_User, time=-0.0046, value=1.0113}
          {arriveTangent=-0.2496, interpMode=RCIM_Cubic, leaveTangent=-0.2496, tangentMode=RCTM_User, time=0.5, value=0.16}
         ]}}}
      Scale Curve=-100
     )
    ))
   &ScaleSpriteSize(Scale Factor=&Vector2DFromFloat(Value=&FloatFromCurve(FloatCurve=DataInterface() {propertyValues {Curve {keys [
          {arriveTangent=2.4634, interpMode=RCIM_Cubic, leaveTangent=2.4634, tangentMode=RCTM_User, time=0, value=0.1}
          {arriveTangent=0.2754, interpMode=RCIM_Cubic, leaveTangent=0.2754, tangentMode=RCTM_User, time=1.0463, value=1}
         ]}}})))
   &Drag(Drag=1.5, Use Linear Drag=true, Use Rotational Drag=false)
   &GravityForce(Gravity=<0 0 -1920>)
   &SolveForcesAndVelocity(Clamp Velocity=false, Limit Acceleration=false)
   &CameraOffset(Camera Offset Amount=10)
   &SubUVAnimation(
    End Frame:NiagaraInt32=&Add_Integer(A=$Particles.SubUVRandomStart, B=23)
    Start Frame:NiagaraInt32=$Particles.SubUVRandomStart
    SubUV Animation Mode:ENiagara_SubUVLookupMode=::NewEnumerator1
    SubUV Curve=DataInterface() {propertyValues {Curve {keys [
        {arriveTangent=1.6197, interpMode=RCIM_Cubic, leaveTangent=1.6197, tangentMode=RCTM_User, time=0, value=0}
        {interpMode=RCIM_Cubic, tangentMode=RCTM_User, time=1, value=1}
       ]}}}
   )
  ]
  rendererClasses [&NiagaraSpriteRendererProperties]
  renderers [{rendererClass=&NiagaraSpriteRendererProperties}]
 }
}
summary {
 systemName=NS_Traction_Exhaust
 emitters [
  {emitterName=Smoke_Wisp, rendererClasses [&NiagaraSpriteRendererProperties]}
  {emitterName=Grass, rendererClasses [&NiagaraMeshRendererProperties]}
  {emitterName=Grass001, simTarget=GPUComputeSim, rendererClasses [&NiagaraSpriteRendererProperties]}
 ]
 userVariables [
  User.bIsFrontEnd {type=NiagaraBool}
  User.bIsFrontEndPreview {type=NiagaraBool}
  User.bIsInWater {type=NiagaraBool}
  User.PlayerSpeed {type=NiagaraFloat, defaultValue=0, description=""}
  User.SkeletalMesh {
   type=NiagaraDataInterfaceSkeletalMesh
   defaultValue {
    struct=NiagaraExt_VariableValue_DataInterface
    value {dataInterface=&self:NiagaraDataInterfaceSkeletalMesh_0}
   }
  }
 ]
}
