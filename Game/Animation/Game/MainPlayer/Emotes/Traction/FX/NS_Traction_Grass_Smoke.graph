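# Niagara system for the MainPlayer Traction emote that kicks up tinted grass mesh debris; despite the Smoke name, it holds only the Grass_Looping emitter.
# Spawning is per unit moved, zeroed by User.bIsInWater, plus a constant rate when User.bIsFrontEndPreview is true.
# Particles take one of two mesh indices and System.GrassBaseColor, burst up and outward from a flattened sphere, then fall under drag and gravity while a curve pops and shrinks their size.
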NiagaraSystem /Game/Animation/Game/MainPlayer/Emotes/Traction/FX/NS_Traction_Grass_Smoke
dependencies {
 usedDataInterfaces [&NiagaraDataInterfaceCurve]
 usedDynamicInputs []
 usedModules [
  &self:SystemSpawnScript.SystemScriptSource.SystemScriptGraph.NiagaraNodeAssignment_0.SetVariables_#1
  &SystemState
 ]
}
emitters {Grass_Looping {
  emitterUpdateScript [
   &EmitterState()
   &SpawnRate(
    SpawnRate=&MakeCustomFloatFromBool(Bool=$User.bIsFrontEndPreview, True Float=20, False Float=0)
    Use Spawn Probability=false
   ) {module=SpawnRate001}
   &SpawnPerUnit(
    Module.Use Max Movement Threshold=true
    Module.Use Movement Tolerance=true
    Movement Tolerance=0.5
    Spawn Per Unit:NiagaraFloat=&Multiply_Float(A=&MakeCustomFloatFromBool(Bool=$User.bIsInWater, True Float=0, False Float=1), B=5)
    Use Spawn Probability=false
   ) {module=SpawnPerUnit001}
  ]
  particleSpawnScript [
   /SetVariables(Particles.MeshIndex=&RandomRangeInt(Minimum=0, Maximum=1)) {isSetParameters=true}
   &InitializeParticle(
    Color=$System.GrassBaseColor
    Lifetime Max=1
    Lifetime Min=0.5
    Lifetime Mode=::NewEnumerator1
    Mesh Scale Max=<0.3 0.5 0.5>
    Mesh Scale Min=<0.1 0.2 0.2>
    Mesh Scale Mode=::NewEnumerator2
    UsePositionOffset=false
    Write Material Random=true
   )
   &InitialMeshOrientation(Use Rotation Vector=false)
   &SphereLocation(
    MaskSpawnGroup=false
    Non Uniform Scale=<1 1 0.1>
    Override Seed=false
    Sphere Radius=15
    Surface Only=false
    UseNonUniformScale=true
    UseOffset=false
    UseSphereOrientation=false
   )
   &AddVelocity(Velocity=&RandomRangeVector(Minimum=<-50 0 100>, Maximum=<50 0 200>))
   &AddVelocityFromPoint(Module.Use Falloff Distance=false, Velocity Strength=200)
  ]
  particleUpdateScript [
   &InheritVelocity(Inherited Velocity Amount Scale=<0.1 0.1 0.1>, Speed Limit=true, UseSpeedThreshold=false)
   &Drag(Use Linear Drag=true, Use Rotational Drag=false)
   &GravityForce()
   &SolveForcesAndVelocity(Clamp Velocity=false, Limit Acceleration=false)
   &ParticleState()
   &ScaleMeshSize(Scale Factor=&VectorFromFloat(Value=&FloatFromCurve(FloatCurve=DataInterface() {propertyValues {Curve {keys [
          {arriveTangent=5, interpMode=RCIM_Cubic, leaveTangent=5, time=0, value=0}
          {arriveTangent=1.0417, interpMode=RCIM_Cubic, leaveTangent=1.0417, time=0.2, value=1}
          {arriveTangent=-1.25, interpMode=RCIM_Cubic, leaveTangent=-1.25, time=0.6, value=0.625}
          {arriveTangent=0.095, interpMode=RCIM_Cubic, leaveTangent=0.095, tangentMode=RCTM_User, time=1, value=0}
         ]}}})))
   &UpdateMeshOrientation(Rotation Rate=&RandomRangeFloat(Minimum=-0.4, Maximum=0.4), Rotation Vector=<1 1 1>)
  ]
  rendererClasses [&NiagaraMeshRendererProperties]
  renderers [{rendererClass=&NiagaraMeshRendererProperties, rendererIndex=0}]
 }}
summary {
 systemName=NS_Traction_Grass_Smoke
 emitters [{emitterName=Grass_Looping, rendererClasses [&NiagaraMeshRendererProperties]}]
 userVariables [
  User.bIsFrontEnd {type=NiagaraBool}
  User.bIsFrontEndPreview {type=NiagaraBool}
  User.bIsInWater {type=NiagaraBool}
  User.PlayerSpeed {type=NiagaraFloat, defaultValue=0, description=""}
 ]
}
